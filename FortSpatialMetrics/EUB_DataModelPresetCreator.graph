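# Editor utility blueprint that creates a spatial metrics data model preset when its Run event fires.
# It calls CreateDataModelPreset with the Asteria parent tree actor class and StaticMesh as the asset dependency class.
# The preset tracks asset instance metrics for instance count, level name and disk size, with all inputs hardcoded in the graph.
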EditorUtilityBlueprint /FortSpatialMetrics/EUB_DataModelPresetCreator
graphs [EventGraph {nodes [
   Event Run()
   > CreateDataModelPreset!(
    execute
    Name:String
    ActorClasses:Array of Actor Soft Class References=Make Array("[0]":Actor Soft Class Reference=/Game/Environments/Asteria/Foliage/Trees/Blueprint_Parent/BP_Tree_Asteria_Parent.BP_Tree_Asteria_Parent_C) -> (Array:Array of Actor Soft Class References)
    AssetDependencies:Array of Object Soft Class References=Make Array("[0]":Object Soft Class Reference=&StaticMesh) -> (Array:Array of Object Soft Class References)
    Metrics:Array of Spatial Metrics Asset Instance Metric Class References=Make Array!(
     "[0]":Spatial Metrics Asset Instance Metric Class Reference=/Script/EditorSpatialMetrics.SpatialMetricsAssetInstance_InstanceCountMetric
     "[1]":Spatial Metrics Asset Instance Metric Class Reference=/Script/EditorSpatialMetrics.SpatialMetricsAssetInstance_LevelNameMetric
     "[2]":Spatial Metrics Asset Instance Metric Class Reference=/Script/EditorSpatialMetrics.SpatialMetricsAssetInstance_DiskSizeMetric
    ) -> (Array:Array of Spatial Metrics Asset Instance Metric Class References)
   ) -> (then, ReturnValue:Object Class Reference)
  ]}]
variables []
